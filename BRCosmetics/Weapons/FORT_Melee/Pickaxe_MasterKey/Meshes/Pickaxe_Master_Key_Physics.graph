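# Physics asset for the Master Key pickaxe melee weapon, giving its dangling ribbon a simulation setup.
# The two-handed handle bone carries two capsules, a long shaft and a shorter one at one end.
# Seven dyn_ribbon bones each hold a small capsule, growing thicker along the chain.
# The first ribbon joint to the handle allows wide swing and twist, while each later link is tightly limited so the ribbon bends stiffly.
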PhysicsAsset /BRCosmetics/Weapons/FORT_Melee/Pickaxe_MasterKey/Meshes/Pickaxe_Master_Key_Physics
bodies {
 Melee_TwoHanded_Handle [
  {shapeType=Capsule, length=104.7886, radius=1.8869, center=<0.8476 -1.1087 22.4924>, rotation=<rot -0.0499 -90 -0.0541>}
  {shapeType=Capsule, length=8.25, radius=3.75, center=<0.448 -1.3668 -37.2868>, rotation=<rot 0 0 -89.9997>}
 ]
 dyn_ribbon_01 [{shapeType=Capsule, length=1.25, radius=1.5, center=<0.7359 0.0071 -0.0608>, rotation=<rot -90 1.8985 -1.8984>}]
 dyn_ribbon_02 [{shapeType=Capsule, length=2.75, radius=1.75, center=<0.4155 0.0085 -0.0575>, rotation=<rot -90 -0.0074 0.0074>}]
 dyn_ribbon_03 [{shapeType=Capsule, length=2.5, radius=2, center=<0.4164 -0.0013 -0.0511>, rotation=<rot -90 -179.9802 179.9803>}]
 dyn_ribbon_04 [{shapeType=Capsule, length=2, radius=2.5, center=<0.4176 -0.0002 -0.0407>, rotation=<rot -90 -1.8984 1.8984>}]
 dyn_ribbon_05 [{shapeType=Capsule, length=1.75, radius=2.75, center=<0.4179 0 -0.0373>, rotation=<rot -90 -7.5524 7.5525>}]
 dyn_ribbon_06 [{shapeType=Capsule, length=1.25, radius=3.25, center=<0.4178 -0.0004 -0.0383>, rotation=<rot -90 -0.844 0.844>}]
 dyn_ribbon_07 [{shapeType=Capsule, length=1.25, radius=4, center=<0.4178 -0.0004 -0.0383>, rotation=<rot -90 -0.844 0.844>}]
}
constraints [
 {bone1Name=dyn_ribbon_01, bone2Name=Melee_TwoHanded_Handle, swing1LimitDegrees=30, swing2LimitDegrees=30, twistLimitDegrees=90}
 {bone1Name=dyn_ribbon_02, bone2Name=dyn_ribbon_01, swing1LimitDegrees=5, swing2LimitDegrees=5, twistLimitDegrees=5}
 {bone1Name=dyn_ribbon_03, bone2Name=dyn_ribbon_02, swing1LimitDegrees=5, swing2LimitDegrees=5, twistLimitDegrees=5}
 {bone1Name=dyn_ribbon_04, bone2Name=dyn_ribbon_03, swing1LimitDegrees=5, swing2LimitDegrees=5, twistLimitDegrees=5}
 {bone1Name=dyn_ribbon_05, bone2Name=dyn_ribbon_04, swing1LimitDegrees=5, swing2LimitDegrees=5, twistLimitDegrees=5}
 {bone1Name=dyn_ribbon_06, bone2Name=dyn_ribbon_05, swing1LimitDegrees=5, swing2LimitDegrees=5, twistLimitDegrees=5}
 {bone1Name=dyn_ribbon_07, bone2Name=dyn_ribbon_06, swing1LimitDegrees=5, swing2LimitDegrees=5, twistLimitDegrees=5}
]
